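# Material function that builds UV sets for a four-layer trim-sheet material, selecting the layer by the U coordinate of UV channel 1.
# Each layer multiplies UV0 by a whole-number scale and shifts V in quarter steps to choose a trim.
# Nested Ifs merge them into combined UVs, output alongside each layer's UVs.
# A parallax occlusion branch with dithered step counts takes those UVs but feeds no output, including the POM-named one.
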MaterialFunction /FigmentEnvironment/LM_8_1_1/Athena/WinterArt/Materials/MF_UVsSet
graph {nodes [
  n1 = ComponentMask(None=TextureCoordinate() {coordinateIndex=1}) {r=true}
  n2 = Constant() {r=1}
  n3 = Ceil(None=ScalarParameter() {defaultValue=1, group=Material_1, parameterName=Mat1 UV Scale, sortPriority=3})
  n4 = Ceil(None=Add(B=ScalarParameter() {defaultValue=1, group=Material_1, parameterName=Mat 1 OffsetTrim 1-4, sliderMax=4, sliderMin=1, sortPriority=5}) {constA=-1})
  n5 = Add(
   A=Multiply(A=TextureCoordinate(), B=Clamp(None=@n3) {maxDefault=1000, minDefault=1})
   B=Multiply(A=Constant2Vector() {g=1}, B=Divide(A=@n4) {constB=4})
  )
  n6 = Ceil(None=ScalarParameter() {defaultValue=1, group=Material_2, parameterName="Mat 2 UV Scale ", sortPriority=4})
  n7 = Ceil(None=Add(B=ScalarParameter() {defaultValue=1, group=Material_2, parameterName=Mat 2 OffsetTrim 1-4, sliderMax=4, sliderMin=1, sortPriority=5}) {constA=-1})
  n8 = Add(
   A=Multiply(A=TextureCoordinate(), B=Clamp(None=@n6) {maxDefault=1000, minDefault=1})
   B=Multiply(A=Constant2Vector() {g=1}, B=Divide(A=@n7) {constB=4})
  )
  n9 = Ceil(None=ScalarParameter() {defaultValue=1, group=Material_3, parameterName="Mat 3 UV Scale ", sortPriority=4})
  n10 = Ceil(None=Add(B=ScalarParameter() {defaultValue=1, group=Material_3, parameterName=Mat 3 OffsetTrim 1-4, sliderMax=4, sliderMin=1, sortPriority=5}) {constA=-1})
  n11 = Add(
   A=Multiply(A=TextureCoordinate(), B=Clamp(None=@n9) {maxDefault=1000, minDefault=1})
   B=Multiply(A=Constant2Vector() {g=1}, B=Divide(A=@n10) {constB=4})
  )
  n12 = Ceil(None=ScalarParameter() {defaultValue=1, group=Material_4, parameterName="Mat 4 UV Scale ", sortPriority=4})
  n13 = Ceil(None=Add(B=ScalarParameter() {defaultValue=1, group=Material_4, parameterName=Mat 4 OffsetTrim 1-4, sliderMax=4, sliderMin=1, sortPriority=5}) {constA=-1})
  n14 = Add(
   A=Multiply(A=TextureCoordinate(), B=Clamp(None=@n12) {maxDefault=1000, minDefault=1})
   B=Multiply(A=Constant2Vector() {g=1}, B=Divide(A=@n13) {constB=4})
  )
  n15 = If(
   A=Constant() {r=3}
   B=@n1
   "A > B"=If(A=Constant() {r=2}, B=@n1, "A > B"=If(A=@n2, B=@n1, "A > B"=@n5, "A < B"=@n8) {desc=0}, "A < B"=@n11)
   "A < B"=@n14
  )
  FunctionOutput(None=FunctionOutput(None=@n15) {outputName=UVs 4 Layers}) {outputName=POM UVs 4 Layers, sortPriority=1}
  FunctionOutput(None=@n5) {outputName=UVs Layer 1}
  FunctionOutput(None=@n8) {outputName=UVs Layer 2}
  FunctionOutput(None=@n11) {outputName=UVs Layer 3}
  FunctionOutput(None=@n14) {outputName=UVs Layer 4}
  n16 = MaterialFunctionCall!(
   Min=ScalarParameter() {defaultValue=8, group=POM, parameterName=POM_MixSteps, sortPriority=13}
   Max=ScalarParameter() {defaultValue=16, group=POM, parameterName=POM_MaxSteps, sortPriority=14}
  ) {materialFunction=&/MF_POMDepth}
  n17 = ScalarParameter() {defaultValue=5, group=POM, parameterName=POM_AAStepsMult, sortPriority=15}
  n18 = MaterialFunctionCall() {materialFunction=&DitherTemporalAA}
  MaterialFunctionCall!(
   Heightmap Texture=FunctionInput() {description="A=HeightMap", inputName=D, inputType=FunctionInput_Texture2D}
   Height Ratio=ScalarParameter() {defaultValue=0.04, group=POM, parameterName=POM_HeightRatio, sortPriority=12}
   Min Steps=@n16.Min
   Max Steps=Multiply(A=@n16.Max, B=LinearInterpolate(B=@n17, Alpha=@n18.Result))
   UVs=@n15
   Heightmap Channel=Constant4Vector() {constant=<rgba 0 0 0 1>}.RGBA
   Reference Plane
   -----------------------------
   Use World Coordinates
   Specify Manual Texture Size
   Manual Texture Size
   ------------------------------
   "Render Shadows (Occlusion Mapping)"
   Light Vector
   Shadow Steps
   Shadow Penumbra
   Transform To VertexNormal
  ) {materialFunction=/Engine/Functions/Engine_MaterialFunctions01/Texturing/ParallaxOcclusionMapping}
 ]}
