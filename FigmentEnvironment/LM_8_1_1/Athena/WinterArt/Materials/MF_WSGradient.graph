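# Material function that tints a material's base color with a world-space vertical gradient between bottom and top colors, behind a static switch.
# MF_WS_GridUV's basement world height is normalized over a rounded-up number of fixed-height floors, with an HLOD offset added for proxy materials.
# A transitionFactor input, defaulting to UV1's red channel, selects one of four per-material ramp values that fades the multiplied tint from white to the full gradient.
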MaterialFunction /FigmentEnvironment/LM_8_1_1/Athena/WinterArt/Materials/MF_WSGradient
graph {nodes [
  ComponentMask(None=WorldPosition().XYZ) {b=true}
  n1 = FunctionInput() {inputName=MaterialIn, inputType=FunctionInput_MaterialAttributes}
  n2 = GetMaterialAttributes(None=@n1) {attributeGetTypes [#1]}
  n3 = VectorParameter() {group=Global, parameterName=WS GradientBottom, sortPriority=1, defaultValue=<rgba 0.4824 0.6547 0.7917 1>}
  n4 = VectorParameter() {group=Global, parameterName=WS GradientTop, sortPriority=1, defaultValue=<rgba 1 1 1 1>}
  n5 = MaterialFunctionCall!(Floor1, Floor2, Floor3, Basement) {materialFunction=&/MF_WS_GridUV}
  n6 = Add(A=@n5.BasementWorldHeight, B=ScalarParameter() {parameterName=ProxyHLODOffset})
  n7 = Constant() {r=384}
  n8 = Ceil(None=ScalarParameter() {defaultValue=1, group=Global, parameterName=GradientFloors, sortPriority=3})
  n9 = Saturate(None=Divide(A=MaterialProxyReplace(Realtime=@n5.BasementWorldHeight, MaterialProxy=@n6), B=Multiply(A=@n7, B=@n8)))
  n10 = Constant() {r=3}
  n11 = FunctionInput(Preview=ComponentMask(None=TextureCoordinate() {coordinateIndex=1}) {r=true}) {bUsePreviewValueAsDefault=true, inputName=transitionFactor, inputType=FunctionInput_Scalar, sortPriority=-1}
  n12 = If(
   A=Constant() {r=2}
   B=@n11
   "A > B"=If(
    A=Constant() {r=1}
    B=@n11
    "A > B"=ScalarParameter() {defaultValue=1, group=Material_1, parameterName=Mat1 Ramp}
    "A < B"=ScalarParameter() {defaultValue=1, group=Material_2, parameterName=Mat2 Ramp}
   ) {desc=0}
   "A < B"=ScalarParameter() {defaultValue=1, group=Material_3, parameterName=Mat3 Ramp}
  )
  n13 = ScalarParameter() {defaultValue=1, group=Material_4, parameterName=Mat4 Ramp}
  n14 = Multiply(
   A=@n2.BaseColor
   B=LinearInterpolate(B=LinearInterpolate(A=@n3.RGB, B=@n4.RGB, Alpha=@n9), Alpha=If(A=@n10, B=@n11, "A > B"=@n12, "A < B"=@n13)) {constA=1}
  )
  FunctionOutput(None=StaticSwitchParameter(
    True=SetMaterialAttributes!(MaterialAttributes=@n2.MaterialAttributes, BaseColor=@n14) {attributeSetTypes [#1]}
    False=@n1
   ) {group=Global, parameterName=Use WS Gradient}) {outputName=Color*Gradient}
 ]}
