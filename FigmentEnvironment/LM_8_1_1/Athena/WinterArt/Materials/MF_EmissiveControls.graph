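# Material function in the Athena WinterArt set that computes emissive output from mask, emissive color and base color inputs.
# A view-angle term from camera vector dot vertex normal feeds Min/Max parameters through SmoothStep when HQEmissive is on, otherwise a lerp.
# A static switch optionally tints emissive color by base color.
# The product is scaled through Color_Time_of_Day, with DayMult shared by morning, day and evening and NightMult for night.
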MaterialFunction /FigmentEnvironment/LM_8_1_1/Athena/WinterArt/Materials/MF_EmissiveControls
graph {nodes [
  n1 = FunctionInput() {inputName=EmissiveMask}
  n2 = DotProduct(A=CameraVectorWS(), B=VertexNormalWS())
  n3 = ScalarParameter() {defaultValue=4, group=Emissive, parameterName=EmissiveFresnel_Min, sortPriority=5}
  n4 = ScalarParameter() {defaultValue=1, group=Emissive, parameterName=EmissiveFresnel_Max, sortPriority=5}
  n5 = StaticSwitchParameter(
   True=MaterialFunctionCall(Alpha=@n2, Min=@n3, Max=@n4) {materialFunction=&SmoothStep}.Result
   False=LinearInterpolate(A=@n3, B=@n4, Alpha=@n2)
  ) {group=Global, parameterName=HQEmissive, sortPriority=3}
  n6 = FunctionInput() {inputName=EmissiveColor}
  n7 = StaticSwitchParameter(True=Multiply(A=FunctionInput() {inputName=BaseColor}, B=@n6), False=@n6) {group=Emissive, parameterName=UseBaseColorForEmissive}
  n8 = ScalarParameter() {defaultValue=1, group=Emissive, parameterName=DayMult, sortPriority=1}
  n9 = MaterialFunctionCall!(
   Color_Morning=@n8
   Color_Day=@n8
   Color_Evening=@n8
   Color_Night=ScalarParameter() {defaultValue=1, group=Emissive, parameterName=NightMult, sortPriority=2}
  ) {materialFunction=/FigmentEnvironment/LM_8_1_1/Effects/Fort_Effects/Materials/_Functions/Color_Time_of_Day}
  FunctionOutput(None=Multiply(A=Multiply(A=@n1, B=@n5), B=Multiply(A=@n7, B=@n9.Result)))
 ]}
