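# Material function for WinterArt materials that blends up to four texture sets into single Diffuse, S, M, R, Emissive, Normals and Height outputs.
# Shared static bools (Use 2/3/4 Materials) drive nested lerp functions for each channel group.
# Diffuse alpha becomes Height, while packed S alpha masks blended emissive colors through MF_EmissiveControls.
# On ES2, the normal blend keeps only the first layer's normal input, substituting flat normals for the others.
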MaterialFunction /FigmentEnvironment/LM_8_1_1/Athena/WinterArt/Materials/MF_FN_TextureSetLerp_4Layers
graph {nodes [
  n1 = Constant4Vector() {constant=<rgba 0.46 0.4009 0.2599 0>}
  n2 = StaticBoolParameter() {group=Material_2, parameterName=Use 2 Materials}
  n3 = StaticBoolParameter() {group=Material_3, parameterName=Use 3 Materials}
  n4 = StaticBoolParameter() {group=Material_4, parameterName=Use 4 Materials}
  n5 = MaterialFunctionCall!(
   transitionFactor
   1=FunctionInput(Preview=@n1.RGBA) {bUsePreviewValueAsDefault=true, inputName=D, inputType=FunctionInput_Vector4}
   2=FunctionInput(Preview=@n1.RGBA) {bUsePreviewValueAsDefault=true, inputName=D2, inputType=FunctionInput_Vector4, sortPriority=3}
   3=FunctionInput(Preview=@n1.RGBA) {bUsePreviewValueAsDefault=true, inputName=D3, inputType=FunctionInput_Vector4, sortPriority=6}
   4=FunctionInput(Preview=@n1.RGBA) {bUsePreviewValueAsDefault=true, inputName=D4, inputType=FunctionInput_Vector4, sortPriority=9}
   Use 2 Textures=@n2
   Use 3 Textures=@n3
   Use 4 Textures=@n4
  ) {materialFunction=&/MF_LerpTexturesBasedOnInt_Float4_4Layers}
  FunctionOutput(None=ComponentMask(None=@n5.Result) {a=true}) {outputName=Height, sortPriority=6}
  n6 = FunctionInput() {bUsePreviewValueAsDefault=true, inputName=S, inputType=FunctionInput_Vector4, sortPriority=1, previewValue=<1 0 0.2667 1>}
  n7 = FunctionInput() {bUsePreviewValueAsDefault=true, inputName=S2, inputType=FunctionInput_Vector4, sortPriority=4, previewValue=<1 0 0.2667 1>}
  n8 = FunctionInput() {bUsePreviewValueAsDefault=true, inputName=S3, inputType=FunctionInput_Vector4, sortPriority=7, previewValue=<1 0 0.2667 1>}
  n9 = FunctionInput() {bUsePreviewValueAsDefault=true, inputName=S4, inputType=FunctionInput_Vector4, sortPriority=11, previewValue=<1 0 0.2667 1>}
  n10 = MaterialFunctionCall(Float4=MaterialFunctionCall(1=@n6, 2=@n7, 3=@n8, 4=@n9, Use 2 Textures=@n2, Use 3 Textures=@n3, Use 4 Textures=@n4) {materialFunction=&MF_LerpTexturesBasedOnInt_Float4_4Layers}.Result) {materialFunction=&BreakOutFloat4Components}
  n11 = FunctionInput() {bUsePreviewValueAsDefault=true, inputName=E1, sortPriority=17}
  n12 = FunctionInput() {bUsePreviewValueAsDefault=true, inputName=E2, sortPriority=18}
  n13 = FunctionInput() {bUsePreviewValueAsDefault=true, inputName=E3, sortPriority=19}
  n14 = FunctionInput() {bUsePreviewValueAsDefault=true, inputName=E4, sortPriority=21}
  n15 = ComponentMask(None=@n5.Result) {b=true, g=true, r=true}
  FunctionOutput(None=MaterialFunctionCall!(
    BaseColor=FunctionOutput(None=@n15) {outputName=Diffuse}
    EmissiveMask=@n10.A
    EmissiveColor=MaterialFunctionCall!(transitionFactor, 1=@n11, 2=@n12, 3=@n13, 4=@n14, Use 2 Textures=@n2, Use 3 Textures=@n3, Use 4 Textures=@n4) {materialFunction=&/MF_LerpTexturesBasedOnInt_Float3_4Layers}.Result
   ) {materialFunction=&/MF_EmissiveControls}.Result) {outputName=Emissive, sortPriority=4}
  n16 = FunctionInput() {bUsePreviewValueAsDefault=true, inputName=N, sortPriority=2, previewValue=<0 0 1 1>}
  n17 = MaterialFunctionCall(
   1=@n16
   2=FunctionInput() {bUsePreviewValueAsDefault=true, inputName=N2, sortPriority=5, previewValue=<0 0 1 1>}
   3=FunctionInput() {bUsePreviewValueAsDefault=true, inputName=N3, sortPriority=8, previewValue=<0 0 1 1>}
   4=FunctionInput() {bUsePreviewValueAsDefault=true, inputName=N4, sortPriority=12, previewValue=<0 0 1 1>}
   Use 2 Textures=@n2
   Use 3 Textures=@n3
   Use 4 Textures=@n4
  ) {materialFunction=&MF_LerpTexturesBasedOnInt_Float3_4Layers}
  n18 = Constant3Vector() {constant=<rgba 0 0 1 0>}
  FunctionOutput(None=FeatureLevelSwitch(
    Default=@n17.Result
    ES2=MaterialFunctionCall(1=@n16, 2=@n18, 3=@n18, 4=@n18, Use 2 Textures=@n2, Use 3 Textures=@n3, Use 4 Textures=@n4) {materialFunction=&MF_LerpTexturesBasedOnInt_Float3_4Layers}.Result
    ES3_1=@n17.Result
    SM4_REMOVED=@n17.Result
    SM5=@n17.Result
    SM6=@n17.Result
   )) {outputName=Normals, sortPriority=5}
  FunctionOutput(None=@n10.R) {outputName=S, sortPriority=1}
  FunctionOutput(None=@n10.G) {outputName=M, sortPriority=2}
  FunctionOutput(None=@n10.B) {outputName=R, sortPriority=3}
 ]}
