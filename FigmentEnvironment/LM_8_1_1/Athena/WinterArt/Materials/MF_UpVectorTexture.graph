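# Material function in the WinterArt set that blends an input texture onto upward-facing surfaces of incoming material attributes.
# The mask scales world vertex normal Z by a strength from HeightLerp, driven by vertex-color blue and a height map input.
# A static switch optionally multiplies in MF_MeshOrientationMask for architectural assets.
# The mask replaces BaseColor and flattens Normal; UseUpVectorTexture off passes the input through.
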MaterialFunction /FigmentEnvironment/LM_8_1_1/Athena/WinterArt/Materials/MF_UpVectorTexture
graph {nodes [
  n1 = FunctionInput() {inputType=FunctionInput_MaterialAttributes}
  n2 = GetMaterialAttributes(None=@n1) {attributeGetTypes [#1, #2]}
  n3 = FunctionInput() {bUsePreviewValueAsDefault=true, inputName=UpVectorTexture, sortPriority=1, previewValue=<0.7 0.7 0.7 1>}
  n4 = VertexNormalWS()
  n5 = Constant3Vector() {constant=<rgba 0 0 1 0>}
  n6 = Constant() {r=0.2}
  n7 = Constant() {r=3}
  n8 = FunctionInput() {bUsePreviewValueAsDefault=true, inputName=BG_HeightMap, sortPriority=1, previewValue=<0.7 0.7 0.7 1>}
  n9 = MaterialFunctionCall(Transition Phase=VertexColor().B, Height Texture=@n8, Contrast=Constant() {r=1}) {materialFunction=&HeightLerp}
  n10 = Saturate(None=Multiply(A=DotProduct(A=@n4, B=@n5), B=LinearInterpolate(A=@n6, B=@n7, Alpha=@n9.Alpha)))
  n11 = StaticSwitchParameter(
   True=Multiply(A=@n10, B=MaterialFunctionCall() {materialFunction=&/MF_MeshOrientationMask}.Result)
   False=@n10
  ) {group=UpVector, parameterName="UseProcedural mask ( ArchAssets)"}
  n12 = LinearInterpolate(A=@n2.BaseColor, B=@n3, Alpha=@n11)
  n13 = LinearInterpolate(A=@n2.Normal, B=Constant3Vector() {constant=<rgba 0 0 1 0>}, Alpha=@n11)
  FunctionOutput(None=StaticSwitchParameter(
    True=SetMaterialAttributes!(MaterialAttributes=@n2.MaterialAttributes, BaseColor=@n12, Normal=@n13) {attributeSetTypes [#1, #2]}
    False=@n1
   ) {group=UpVector, parameterName=UseUpVectorTexture})
 ]}
