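# Material function that layers a vertex-painted texture, snow by default, over incoming material attributes behind a static switch.
# Red vertex paint and per-material spread drive height lerps against BGMap, producing the coverage and a modified outline band; green suppresses the painted normal.
# Textures sample by UV, optionally with parallax occlusion, or world-aligned.
# WinterSnow switches to shared Athena snow and biome color, replacing base color outright, unmasked by paint.
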MaterialFunction /FigmentEnvironment/LM_8_1_1/Athena/WinterArt/Materials/MF_VPTexture
graph {nodes [
  n1 = FunctionInput() {inputType=FunctionInput_MaterialAttributes}
  n2 = GetMaterialAttributes(None=@n1) {attributeGetTypes [
    #1
    #2
    #3
    #4
    #5
    #6
   ]}
  n3 = GetMaterialAttributes(None=MaterialFunctionCall!(
    In=MaterialFunctionCall!(UseLandscapeCoordsOnSwitch) {materialFunction=/FigmentEnvironment/LM_8_1_1/Athena/Environments/Landscape/MaterialFunctions/Snow/MF_Athena_Snow_01}.Material
    Color=MaterialFunctionCall!(Material) {materialFunction=/FigmentEnvironment/LM_8_1_1/Athena/Environments/Materials/MF_StaticMeshBiomeColor}.Color
   ) {materialFunction=/FigmentEnvironment/LM_8_1_1/Athena/Environments/Landscape/MaterialFunctions/MF_Athena_Material_BaseMultiply}.Result) {attributeGetTypes [#2, #3, #1]}
  n4 = VectorParameter() {group=VertexPaintTexture, parameterName=VP_OutlineD_Mod, sortPriority=13, defaultValue=<rgba 0.7 0.7 0.7 0>}
  n5 = StaticBoolParameter() {group=Material_2, parameterName=Use 2 Materials}
  n6 = StaticBoolParameter() {group=Material_3, parameterName=Use 3 Materials}
  n7 = StaticBoolParameter() {group=Material_4, parameterName=Use 4 Materials}
  n8 = VertexColor()
  n9 = Multiply(
   A=MaterialFunctionCall!(
    transitionFactor
    1=ScalarParameter() {defaultValue=1, group=Material_1, parameterName=VP_Spread_1, sliderMax=2, sortPriority=11}
    2=ScalarParameter() {defaultValue=1, group=Material_2, parameterName=VP_Spread_2, sliderMax=2, sortPriority=11}
    3=ScalarParameter() {defaultValue=1, group=Material_3, parameterName=VP_Spread_3, sliderMax=2, sortPriority=11}
    4=ScalarParameter() {defaultValue=1, group=Material_4, parameterName=VP_Spread_4, sliderMax=2, sortPriority=11}
    Use 2 Textures=@n5
    Use 3 Textures=@n6
    Use 4 Textures=@n7
   ) {materialFunction=&/MF_LerpTexturesBasedOnInt_Float4_4Layers}.Result
   B=@n8.R
  )
  n10 = Constant()
  n11 = Constant() {r=0.35}
  n12 = ScalarParameter() {defaultValue=0.7, group=VertexPaintTexture, parameterName=OutlineDistance, sortPriority=4}
  n13 = LinearInterpolate(
   A=ScalarParameter() {group=VertexPaintTexture, parameterName=VP_Background_HM_Min, sortPriority=2}
   B=ScalarParameter() {defaultValue=1, group=VertexPaintTexture, parameterName=VP_Background_HM_Max, sortPriority=1}
   Alpha=FunctionInput() {bUsePreviewValueAsDefault=true, inputName=BGMap, inputType=FunctionInput_Scalar, sortPriority=1}
  )
  n14 = TextureObjectParameter() {group=VertexPaintTexture, parameterName=VP_Color, sortPriority=15, texture=/FigmentEnvironment/LM_8_1_1/Athena/WinterArt/Textures/T_Snow01_D}
  n15 = Multiply(A=ScalarParameter() {defaultValue=1, group=VertexPaintTexture, parameterName=VP_TextureScaleWS, sortPriority=14}, B=Constant() {r=256})
  n16 = MaterialFunctionCall!(
   Min=ScalarParameter() {defaultValue=8, group=Material_4, parameterName=VP_POM_MixSteps, sortPriority=13}
   Max=ScalarParameter() {defaultValue=16, group=Material_4, parameterName=VP_POM_MaxSteps, sortPriority=14}
  ) {materialFunction=&/MF_POMDepth}
  n17 = MaterialFunctionCall() {materialFunction=&DitherTemporalAA}
  n18 = ScalarParameter() {defaultValue=1, group=Material_1, parameterName=VP_TextureScale, sortPriority=15}
  n19 = ScalarParameter() {defaultValue=1, group=Material_2, parameterName=VP_TextureScale2, sortPriority=15}
  n20 = ScalarParameter() {defaultValue=1, group=Material_3, parameterName=VP_TextureScale3, sortPriority=15}
  n21 = ScalarParameter() {defaultValue=1, group=Material_4, parameterName=VP_TextureScale4, sortPriority=15}
  n22 = Multiply(
   A=TextureCoordinate()
   B=MaterialFunctionCall(1=Ceil(None=@n18), 2=Ceil(None=@n19), 3=Ceil(None=@n20), 4=Ceil(None=@n21), Use 2 Textures=@n5, Use 3 Textures=@n6, Use 4 Textures=@n7) {materialFunction=&MF_LerpTexturesBasedOnInt_Float4_4Layers}.Result
  )
  n23 = MaterialFunctionCall!(
   Heightmap Texture=@n14
   Height Ratio=ScalarParameter() {defaultValue=0.04, group=Material_4, parameterName=VP_POM_HeightRatio, sortPriority=12}
   Min Steps=@n16.Min
   Max Steps=Multiply(A=@n16.Max, B=LinearInterpolate(Alpha=@n17.Result))
   UVs=@n22
   Heightmap Channel=Constant4Vector() {constant=<rgba 0 0 0 1>}.RGBA
   Reference Plane=ScalarParameter() {defaultValue=1, group=Material_4, parameterName=VP_POM_Depth, sortPriority=16}
   -----------------------------
   Use World Coordinates
   Specify Manual Texture Size
   Manual Texture Size
   ------------------------------
   "Render Shadows (Occlusion Mapping)"
   Light Vector
   Shadow Steps
   Shadow Penumbra
   Transform To VertexNormal
  ) {materialFunction=/Engine/Functions/Engine_MaterialFunctions01/Texturing/ParallaxOcclusionMapping}
  n24 = TextureSample(UVs=StaticSwitchParameter(True=@n23.Parallax UVs, False=@n22) {group=VertexPaintTexture, parameterName=POM_VP_True}, Tex=@n14) {texture=/FigmentEnvironment/LM_8_1_1/Athena/WinterArt/Textures/T_Snow01_N}
  n25 = StaticSwitchParameter(
   True=MaterialFunctionCall(TextureObject=@n14, TextureSize=@n15, Export Float 4=StaticBool() {value=true}) {materialFunction=&WorldAlignedTexture}.XYZ Texture
   False=AppendVector(A=@n24.RGB, B=@n24.RGB)
  ) {group=VertexPaintTexture, parameterName=WS_VPTexture}
  n26 = ComponentMask(None=@n25) {a=true}
  n27 = Multiply(A=@n26, B=ScalarParameter() {defaultValue=1, group=VertexPaintTexture, parameterName=VP_TextureHeightMax, sortPriority=3})
  n28 = MaterialFunctionCall!(
   Diffuse 1
   Diffuse 2
   Transistion Phase=Subtract(A=@n9, B=LinearInterpolate(A=@n10, B=@n11, Alpha=@n12)) {constB=0.2}
   Height Texture 1=@n27
   Height Texture 2=@n13
   Contrast=ScalarParameter() {defaultValue=3, group=VertexPaintTexture, parameterName=VP_Outline_Contrast, sortPriority=6}
  ) {materialFunction=/Engine/Functions/Engine_MaterialFunctions02/Texturing/HeightLerpWithTwoHeightMaps}
  n29 = LinearInterpolate(A=@n2.BaseColor, B=Multiply(A=@n2.BaseColor, B=@n4.RGB) {constB=0.5}, Alpha=@n28.Alpha)
  n30 = GetMaterialAttributes(None=MaterialFunctionCall(Material=MaterialFunctionCall() {materialFunction=&MF_Athena_Snow_01}.Material) {materialFunction=&MF_StaticMeshBiomeColor}.Material) {attributeGetTypes [#1]}
  n31 = VectorParameter() {group=VertexPaintTexture, parameterName=VP_Tint, sortPriority=13, defaultValue=<rgba 0.75 0.75 0.75 1>}
  n32 = ComponentMask(None=Multiply(
    A=StaticSwitchParameter(True=@n30.BaseColor, False=@n31.RGB) {group=VertexPaintTexture, parameterName=WinterSnow}
    B=ComponentMask(None=@n25) {b=true, g=true, r=true}
   )) {b=true, g=true, r=true}
  n33 = MaterialFunctionCall(
   Transistion Phase=@n9
   Height Texture 1=@n27
   Height Texture 2=@n13
   Contrast=ScalarParameter() {defaultValue=5, group=VertexPaintTexture, parameterName=VP_Contrast, sortPriority=5}
  ) {materialFunction=&HeightLerpWithTwoHeightMaps}
  n34 = TextureObjectParameter() {
   group=VertexPaintTexture
   parameterName=VP_RGBMask
   samplerType=SAMPLERTYPE_Masks
   sortPriority=15
   texture=/FigmentEnvironment/LM_8_1_1/Athena/WinterArt/Textures/T_MaskDef
  }
  n35 = StaticSwitchParameter(True=@n23.Parallax UVs, False=@n22) {group=VertexPaintTexture, parameterName=POM_VP_True}
  n36 = MaterialFunctionCall(Float3=StaticSwitchParameter(
    True=MaterialFunctionCall(TextureObject=@n34, TextureSize=@n15) {materialFunction=&WorldAlignedTexture}.XYZ Texture
    False=TextureSample(UVs=@n35, Tex=@n34) {samplerType=SAMPLERTYPE_Masks, texture=&T_Snow01_N}.RGB
   ) {group=VertexPaintTexture, parameterName=WS_VPTexture}) {materialFunction=&BreakOutFloat3Components}
  n37 = LinearInterpolate(
   A=ScalarParameter() {defaultValue=0.4, parameterName=VPTexture_Spec_Min}
   B=ScalarParameter() {defaultValue=0.6, parameterName=VPTexture_Spec_Max}
   Alpha=@n26
  )
  n38 = LinearInterpolate(
   A=ScalarParameter() {defaultValue=0.4, parameterName=VPTexture_MobileSpec_Min}
   B=ScalarParameter() {defaultValue=0.6, parameterName=VPTexture_MobileSpec_Max}
   Alpha=@n26
  )
  n39 = StaticSwitchParameter(True=@n36.R, False=QualitySwitch(Default=FeatureLevelSwitch(Default=@n37, ES2=@n38, ES3_1=@n38), Low=@n38)) {group=VertexPaintTexture, parameterName=VPRGBMask}
  n40 = Multiply(A=@n2.Roughness, B=ScalarParameter() {defaultValue=0.35, group=VertexPaintTexture, parameterName=VP_OutlineR_Mod, sortPriority=7})
  n41 = StaticSwitchParameter(
   True=@n36.B
   False=LinearInterpolate(
    A=ScalarParameter() {defaultValue=0.5, group=VertexPaintTexture, parameterName=VPTexture_R_Min, sortPriority=8}
    B=ScalarParameter() {defaultValue=0.5, group=VertexPaintTexture, parameterName=VPTexture_R_Max, sortPriority=9}
    Alpha=ComponentMask(None=@n25) {r=true}
   )
  ) {group=VertexPaintTexture, parameterName=VPRGBMask}
  n42 = OneMinus(None=@n8.G)
  n43 = Constant3Vector() {constant=<rgba 0 0 1 0>}
  n44 = TextureObjectParameter() {
   group=VertexPaintTexture
   parameterName=VP_Normal
   samplerType=SAMPLERTYPE_Normal
   sortPriority=16
   texture=&T_Snow01_N
  }
  n45 = TextureSample!(UVs=@n23.Parallax UVs, Tex=@n44, "DDX(UVs)"=DDX(Value=@n22), "DDY(UVs)"=DDY(Value=@n22), Apply View MipBias) {mipValueMode=TMVM_Derivative, samplerType=SAMPLERTYPE_Normal, texture=&T_Snow01_N}
  n46 = StaticSwitchParameter(
   True=MaterialFunctionCall(TextureObject=@n44, TextureSize=@n15) {materialFunction=&WorldAlignedNormal}.XYZ Texture
   False=StaticSwitchParameter(
    True=@n45.RGB
    False=TextureSample!(UVs=@n22, Tex=@n44, Apply View MipBias) {samplerType=SAMPLERTYPE_Normal, texture=&T_Snow01_N}.RGB
   ) {group=VertexPaintTexture, parameterName=POM_VP_True}
  ) {group=VertexPaintTexture, parameterName=WS_VPTexture}
  n47 = MaterialFunctionCall(
   1=ScalarParameter() {group=Material_1, parameterName=M1_VP_FlattenNormal, sortPriority=17}
   2=ScalarParameter() {group=Material_2, parameterName=M2_VP_FlattenNormal, sortPriority=17}
   3=ScalarParameter() {group=Material_3, parameterName=M3_VP_FlattenNormal, sortPriority=17}
   4=ScalarParameter() {group=Material_4, parameterName=M4_VP_FlattenNormal, sortPriority=17}
   Use 2 Textures=@n5
   Use 3 Textures=@n6
   Use 4 Textures=@n7
  ) {materialFunction=&MF_LerpTexturesBasedOnInt_Float4_4Layers}
  n48 = Abs(None=VertexNormalWS())
  n49 = Multiply(A=Multiply(A=ComponentMask(None=@n48) {r=true}, B=ComponentMask(None=@n48) {g=true}), B=Constant() {r=3})
  n50 = StaticSwitchParameter(
   True=@n43
   False=LinearInterpolate(
    A=MaterialFunctionCall(Normal=@n46, Flatness=@n47.Result) {materialFunction=&FlattenNormal}.Result
    B=@n43
    Alpha=Saturate(None=@n49)
   )
  ) {group=VertexPaintTexture, parameterName=FlatNormal, sortPriority=11}
  n51 = SetMaterialAttributes!(
   MaterialAttributes=@n2.MaterialAttributes
   BaseColor=StaticSwitchParameter(True=@n3.BaseColor, False=LinearInterpolate(A=@n29, B=@n32, Alpha=@n33.Alpha)) {group=VertexPaintTexture, parameterName=WinterSnow}
   Specular=LinearInterpolate(
    A=@n2.Specular
    B=StaticSwitchParameter(True=@n3.Specular, False=@n39) {group=VertexPaintTexture, parameterName=WinterSnow}
    Alpha=@n33.Alpha
   )
   Roughness=LinearInterpolate(
    A=LinearInterpolate(A=@n2.Roughness, B=@n40, Alpha=@n28.Alpha)
    B=StaticSwitchParameter(True=@n3.Roughness, False=@n41) {group=VertexPaintTexture, parameterName=WinterSnow}
    Alpha=@n33.Alpha
   )
   Normal=LinearInterpolate(A=@n2.Normal, B=QualitySwitch(Default=@n50, Low=@n43, High=@n50, Medium=@n50), Alpha=Multiply(A=@n33.Alpha, B=@n42))
   Metallic=LinearInterpolate(A=LinearInterpolate(A=@n2.Metallic, Alpha=@n28.Alpha) {constB=0.2}, Alpha=@n33.Alpha) {constB=0}
   EmissiveColor=LinearInterpolate(A=@n2.EmissiveColor, Alpha=@n33.Alpha) {constB=0}
  ) {attributeSetTypes [
    #1
    #2
    #3
    #4
    #5
    #6
   ]}
  FunctionOutput(None=StaticSwitchParameter(True=@n51, False=@n1) {desc=Red Paint Green Blur, group=VertexPaintTexture, parameterName="UseVP Texture (R Paint, G ActorNMap)"})
  FunctionOutput(None=@n32) {outputName=VP_Color, sortPriority=2}
 ]}
